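; PCG example graph that builds a forest with a clearing, named as a 200m square with a circular clearing.
; Two Create_Spline_Shape primitives feed Compose_Areas_Exclusion as primary area and excluded area.
; The resulting area goes through Fill_Areas_Randomly into Spawn_Assets for scattered placement.
; The primary shape also feeds Create_Mesh_Planar directly, so the planar mesh bypasses the exclusion.
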(graph Forest_Square_200m_With_Circled_Clearing
  (bind Subgraph_0 (Subgraph :Graph /PCGPrimitives/Primitives/Create/Create_Spline_Shape.Create_Spline_Shape))
  (bind Subgraph_1 (Subgraph :Graph /PCGPrimitives/Primitives/Create/Create_Spline_Shape.Create_Spline_Shape))
  (bind Subgraph_6 (Subgraph :Graph /PCGPrimitives/Primitives/Compose/Compose_Areas_Exclusion.Compose_Areas_Exclusion))
  (bind Subgraph_7 (Subgraph :Graph /PCGPrimitives/Primitives/Fill/Fill_Areas_Randomly.Fill_Areas_Randomly))
  (bind Subgraph_8 (Subgraph :Graph /PCGPrimitives/Primitives/Spawn/Spawn_Assets.Spawn_Assets))
  (bind Subgraph_9 (Subgraph :Graph /PCGPrimitives/Primitives/Create/Create_Mesh_Planar.Create_Mesh_Planar))
  (connect Subgraph_0 Out Subgraph_6 PrimaryInput)
  (connect Subgraph_1 Out Subgraph_6 SecondaryInput)
  (connect Subgraph_6 Out Subgraph_7 PrimaryInput)
  (connect Subgraph_7 Out Subgraph_8 PrimaryInput)
  (connect Subgraph_0 Out Subgraph_9 PrimaryInput))
