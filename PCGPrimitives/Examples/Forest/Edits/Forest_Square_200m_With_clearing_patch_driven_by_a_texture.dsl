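; PCG example graph that builds a square forest, 200m by its name, with clearing patches driven by a texture, composed entirely of PCGPrimitives subgraphs.
; A spline shape is filled with random points, and a top-down projected texture writes its value to each point's $Density.
; An attribute value filter passes only the selected points to asset spawning, forming the clearings.
; The same spline also feeds a planar mesh creation branch.
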(graph Forest_Square_200m_With_clearing_patch_driven_by_a_texture
  (bind Subgraph_0 (Subgraph :Graph /PCGPrimitives/Primitives/Create/Create_Spline_Shape.Create_Spline_Shape))
  (bind Subgraph_4 (Subgraph :Graph /PCGPrimitives/Primitives/Fill/Fill_Areas_Randomly.Fill_Areas_Randomly))
  (bind Subgraph_1 (Subgraph :Graph /PCGPrimitives/Primitives/Write/Write_Attribute_fromTexture.Write_Attribute_fromTexture)) ; A texture is projected top down onto the points and write \nthe texture value to the $Density attribute
  (bind Subgraph_6 (Subgraph :Graph /PCGPrimitives/Primitives/Filter/Filter_AttributeValue.Filter_AttributeValue))
  (bind Subgraph_8 (Subgraph :Graph /PCGPrimitives/Primitives/Spawn/Spawn_Assets.Spawn_Assets))
  (bind Subgraph_9 (Subgraph :Graph /PCGPrimitives/Primitives/Create/Create_Mesh_Planar.Create_Mesh_Planar))
  (connect Subgraph_4 Out Subgraph_1 PrimaryInput)
  (connect Subgraph_0 Out Subgraph_4 PrimaryInput)
  (connect Subgraph_1 Out Subgraph_6 PrimaryInput)
  (connect Subgraph_6 Selected Subgraph_8 PrimaryInput)
  (connect Subgraph_0 Out Subgraph_9 PrimaryInput))
