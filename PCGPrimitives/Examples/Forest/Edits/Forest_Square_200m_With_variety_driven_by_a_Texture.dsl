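; PCG example graph that builds a square forest area, 200m by its name, with asset variety driven by a texture.
; A spline shape feeds both a planar mesh and a random area fill that generates points.
; A texture projected top-down onto those points writes a remapped TreeIndex attribute, and the points then pass to asset spawning.
; Every step is a reusable PCGPrimitives subgraph.
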(graph Forest_Square_200m_With_variety_driven_by_a_Texture
  (bind Subgraph_0 (Subgraph :Graph /PCGPrimitives/Primitives/Create/Create_Spline_Shape.Create_Spline_Shape))
  (bind Subgraph_6 (Subgraph :Graph /PCGPrimitives/Primitives/Fill/Fill_Areas_Randomly.Fill_Areas_Randomly))
  (bind Subgraph_1 (Subgraph :Graph /PCGPrimitives/Primitives/Write/Write_Attribute_fromTexture.Write_Attribute_fromTexture)) ; Here we project a texture onto the poitns from top Daown\nonto a TreeIndex Attribute. Notice how the tree index is \nremapped to the same number of
  (bind Subgraph_7 (Subgraph :Graph /PCGPrimitives/Primitives/Spawn/Spawn_Assets.Spawn_Assets))
  (bind Subgraph_8 (Subgraph :Graph /PCGPrimitives/Primitives/Create/Create_Mesh_Planar.Create_Mesh_Planar))
  (connect Subgraph_6 Out Subgraph_1 PrimaryInput)
  (connect Subgraph_0 Out Subgraph_6 PrimaryInput)
  (connect Subgraph_1 Out Subgraph_7 PrimaryInput)
  (connect Subgraph_0 Out Subgraph_8 PrimaryInput))
